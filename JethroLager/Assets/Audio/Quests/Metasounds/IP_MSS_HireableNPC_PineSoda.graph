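# MetaSound Source preset that plays a music stinger for the PineSoda hireable NPC, filed under quest audio.
# It inherits its graph from MSS_SingleStinger and supplies a single "Hire Chewy" stinger wave through the In Array input.
# It is a stereo one-shot source that exposes an OnFinished trigger and routes to the Fort_Music_Stinger_PSM sound class.
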MetaSoundSource /JethroLager/Assets/Audio/Quests/Metasounds/IP_MSS_HireableNPC_PineSoda
preset_parent=/FortMetasound/Music/MSS_SingleStinger
inputs [
 In Array {type=WaveAsset:Array}
 UE.Source.OnPlay {type=Trigger}
]
interfaces [UE.OutputFormat.Stereo:1.0, UE.Source.OneShot:1.0, UE.Source:1.1]
outputs [
 "UE.OutputFormat.Stereo.Audio:0" {type=Audio}
 "UE.OutputFormat.Stereo.Audio:1" {type=Audio}
 UE.Source.OneShot.OnFinished {type=Trigger}
]
pages [Default {input_defaults {UE.Source.OnPlay=null, In Array [&/IP_MX_S35_Hire_Chewy_STNG]}}]
properties {
 assetImportData=&self:AssetImportData
 numChannels=2
 outputFormat=Stereo
 sampleRate=48000
 soundClassObject=/Game/Sounds/FortSoundClassesAndModes/Fort_Music_Stinger_PSM
}
